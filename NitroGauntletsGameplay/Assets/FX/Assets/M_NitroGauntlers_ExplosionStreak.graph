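# Translucent unlit Niagara sprite material that draws blurred streaks for the Nitro Gauntlets explosion effect.
# Opacity comes from particle color times a streak texture, with optional depth fade (skipped on mobile, distance optionally from a dynamic parameter) and near-camera fade.
# Emissive uses particle color, with optional dynamic-parameter hue and intensity shifts and a switchable time-of-day tint, fed through a shared FX emissive adjustment function.
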Material /NitroGauntletsGameplay/Assets/FX/Assets/M_NitroGauntlers_ExplosionStreak
graph {
 properties {
  bAllowNegativeEmissiveColor=true
  bUsedWithNanite=true
  bUsedWithNiagaraSprites=true
  blendMode=BLEND_Translucent
  pixelDepthOffsetMode=PDOM_Legacy
  shadingModel=MSM_Unlit
  usedShadingModels=Unlit
 }
 nodes [
  n1 = ParticleColor()
  n2 = Multiply(A=@n1.A, B=TextureSample() {texture=/NitroGameplay/Assets/FX/T_Streak_Blur}.RGB)
  n3 = LinearInterpolate(
   A=ScalarParameter() {defaultValue=5, group=Opacity|DepthFade, parameterName=DepthFade_Distance, sortPriority=2}
   B=DynamicParameter() {defaultValue=<rgba 1 1 10 1>, paramNames [DepthFade,BP_Intensity,BP_HueShift,none]}.DepthFade
   Alpha=ScalarParameter() {group=Opacity|DepthFade, parameterName=Use Dynamic Parameter to Control Depth Fade, sortPriority=1}
  )
  n4 = LinearInterpolate(
   A=@n2
   B=ShadingPathSwitch(Default=DepthFade(Opacity=@n2, FadeDistance=@n3), Mobile=@n2)
   Alpha=ScalarParameter() {group=Opacity|DepthFade, parameterName=Use Depth Fade}
  )
  n5 = Power(
   Base=MaterialFunctionCall(Distance=ScalarParameter() {defaultValue=512, group=Opacity|Near_Camera_Fade, parameterName=NearCamFade_Distance, sortPriority=1}) {materialFunction=&Near_Camera_Fade}.Result
   Exp=ScalarParameter() {defaultValue=1, group=Opacity|Near_Camera_Fade, parameterName=NearCamFade_Power, sortPriority=2}
  )
  LinearInterpolate(
   A=@n4
   B=Multiply(A=@n4, B=Clamp(None=@n5) {minDefault=0.02})
   Alpha=ScalarParameter() {group=Opacity|Near_Camera_Fade, parameterName=Add Near Camera Fade}
  )
  n6 = MaterialFunctionCall(
   Color_Morning=VectorParameter() {parameterName=Color Morning, defaultValue=<rgba 0.2795 0.1398 0.1407 1>}.RGB
   Color_Day=VectorParameter() {parameterName=Color Day, defaultValue=<rgba 0.6094 0.545 0.5552 1>}.RGB
   Color_Evening=VectorParameter() {parameterName=Color Evening, defaultValue=<rgba 0.2726 0.1761 0.1282 1>}.RGB
   Color_Night=VectorParameter() {parameterName=Color Night, defaultValue=<rgba 0.3438 0.3438 0.3438 1>}.RGB
  ) {materialFunction=&Color_Time_of_Day}
  n7 = DynamicParameter() {defaultValue=<rgba 1 1 10 1>, paramNames [DepthFade,BP_Intensity,BP_HueShift,none]}
  n8 = LinearInterpolate(
   A=@n1.RGB
   B=Multiply(
    A=@n7.BP_Intensity
    B=MaterialFunctionCall(Hue Shift Percentage=@n7.BP_HueShift, Texture=@n1.RGB) {materialFunction=&HueShift}.Result
   )
   Alpha=ScalarParameter() {group=Emissive_Color, parameterName="Use HUE && INTENSITY Shifts"}
  )
  MaterialFunctionCall(Emissive=StaticSwitchParameter(True=Multiply(A=@n6.Result, B=@n8), False=@n8) {parameterName=UseTimeOfDayColor}) {materialFunction=&MF_FX_EmissiveAdjustments}
 ]
}
