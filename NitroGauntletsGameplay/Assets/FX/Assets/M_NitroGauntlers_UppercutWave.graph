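# Unlit, two-sided translucent Niagara mesh-particle material for the Nitro Gauntlets uppercut wave effect.
# Panning smoke distorts the UVs of noise, plasma and mask textures, whose panners take time from the dynamic parameter "offset".
# An orange-tinted blend, multiplied by particle color, feeds emissive adjustments.
# An alpha chain combines noise, vertical gradients, particle alpha and view-angle fade, while a separate branch scales panning noise along vertex normals for displacement.
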Material /NitroGauntletsGameplay/Assets/FX/Assets/M_NitroGauntlers_UppercutWave
graph {
 properties {
  bAutomaticallySetUsageInEditor=false
  bUsedWithNiagaraMeshParticles=true
  blendMode=BLEND_Translucent
  pixelDepthOffsetMode=PDOM_Legacy
  previewMesh=&/SM_NitroGauntlers_UppercutWave
  shadingModel=MSM_Unlit
  twoSided=true
  usedShadingModels=Unlit
 }
 nodes [
  n1 = ParticleColor()
  n2 = Panner(Coordinate=TextureCoordinate() {vTiling=0.2}) {speedY=0.1}
  n3 = ComponentMask(None=TextureCoordinate()) {g=true}
  n4 = Multiply(A=Add(A=OneMinus(None=@n3)) {constB=0}) {constB=2}
  n5 = Multiply(A=Multiply(A=TextureSample(UVs=@n2) {texture=&T_Inky_Smoke_Tile}.R, B=Saturate(None=@n4))) {constB=0.15}
  n6 = DynamicParameter() {paramNames [offset,Param2,Param3,Param4]}
  n7 = Panner(Coordinate=Add(A=TextureCoordinate() {uTiling=2, vTiling=0.2}, B=@n5) {constB=0}, Time=@n6.offset) {speedY=0.3}
  n8 = TextureCoordinate() {vTiling=0.2}
  n9 = TextureSample(UVs=Panner(Coordinate=Add(A=@n8, B=@n5) {constB=0}, Time=@n6.offset) {speedY=0.3}) {samplerType=SAMPLERTYPE_LinearGrayscale, texture=/Game/Effects/Fort_Effects/Emotes/UltraLight/Texture/T_Plasma_A_Greyscale}
  n10 = LinearInterpolate(
   A=Multiply(
    A=TextureSample(UVs=@n7) {texture=/Game/Weapons/WeaponSkins/Textures/T_Swirly_Noise_highContrast}.R
    B=Constant3Vector() {constant=<rgba 0.5 0.2177 0.0493 1>}
   )
   B=Multiply(A=Constant3Vector() {constant=<rgba 2 0.1 0 1>}, B=Power(Base=@n9.R) {constExponent=1})
  )
  MaterialFunctionCall(Emissive=Multiply(A=@n1.RGB, B=Power(Base=@n10))) {materialFunction=&MF_FX_EmissiveAdjustments}
  n11 = TextureSample(UVs=Panner(Coordinate=TextureCoordinate()) {speedY=1}) {samplerType=SAMPLERTYPE_Grayscale, texture=&LowResBlurredNoise}
  n12 = OneMinus(None=ComponentMask(None=TextureCoordinate()) {g=true})
  n13 = Saturate(None=Multiply(A=Add(A=@n12) {constB=-0.1}) {constB=2})
  Multiply(A=Multiply(A=Multiply(A=@n11.R, B=@n13), B=VertexNormalWS())) {constB=10}
  n14 = TextureCoordinate() {vTiling=0.2}
  n15 = TextureSample(UVs=Panner(Coordinate=Add(A=@n14, B=@n5) {constB=0}, Time=@n6.offset) {speedY=0.2}) {texture=/Game/Weapons/FORT_Melee/Pickaxe_CoatCheck/FX/T_Pickaxe_CoatCheck_Noise}
  n16 = ComponentMask(None=TextureCoordinate()) {g=true}
  n17 = Multiply(A=Add(A=OneMinus(None=@n16)) {constB=-0.05}) {constB=2}
  n18 = Add(A=TextureCoordinate() {uTiling=2, vTiling=0.2}, B=@n5) {constB=0}
  n19 = OneMinus(None=ComponentMask(None=TextureCoordinate()) {g=true})
  n20 = Power(
   Base=TextureSample(UVs=Panner(Coordinate=@n18, Time=@n6.offset) {speedY=0.2}) {samplerType=SAMPLERTYPE_LinearColor, texture=/Game/Effects/Fort_Effects/BluGlow/Violentium/Violentium_Mask}.R
   Exp=LinearInterpolate(Alpha=Add(A=@n19) {constB=0.2}) {constA=-1, constB=1.5}
  )
  n21 = TextureCoordinate()
  n22 = Multiply(A=Add(A=ComponentMask(None=@n21) {g=true}) {constB=-0.3}) {constB=2}
  n23 = Multiply(
   A=Multiply(A=Power(Base=@n15.R) {constExponent=1}, B=Saturate(None=@n17)) {constA=1}
   B=Multiply(A=LinearInterpolate(Alpha=@n20) {constA=-1}, B=Saturate(None=@n22))
  )
  Multiply(
   A=Multiply(A=@n1.A, B=Multiply(A=@n23))
   B=MaterialFunctionCall("FadeSensitivity (optional)"=Constant() {r=2}) {materialFunction=&FadeBasedOnViewAngle}.Result
  )
 ]
}
