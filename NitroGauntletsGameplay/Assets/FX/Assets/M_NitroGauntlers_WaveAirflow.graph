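# Translucent, unlit, two-sided material for Niagara mesh particles, named as the wave airflow effect of the Nitro Gauntlets.
# Particle color feeds an emissive adjustment function.
# The opacity-style mask multiplies two noise and smoke textures panning along V at different speeds, fades both V ends with UV gradients, and scales by particle alpha.
# A separate branch pans a flame texture at low intensity, with no visible link to the others.
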Material /NitroGauntletsGameplay/Assets/FX/Assets/M_NitroGauntlers_WaveAirflow
graph {
 properties {
  bAutomaticallySetUsageInEditor=false
  bUsedWithNiagaraMeshParticles=true
  blendMode=BLEND_Translucent
  pixelDepthOffsetMode=PDOM_Legacy
  previewMesh=/Game/Effects/Fort_Effects/Emotes/DualParadox/SM_Emote_DualParadox_Wave
  shadingModel=MSM_Unlit
  twoSided=true
  usedShadingModels=Unlit
 }
 nodes [
  n1 = ParticleColor()
  MaterialFunctionCall(Emissive=@n1.RGB) {materialFunction=&MF_FX_EmissiveAdjustments}
  n2 = Panner(Coordinate=TextureCoordinate() {uTiling=3, vTiling=0.2}) {speedY=0.5}
  Multiply(A=Power(Base=TextureSample(UVs=@n2) {samplerType=SAMPLERTYPE_LinearGrayscale, texture=/Game/Weapons/FORT_Melee/Pickaxe_Cube_Ninja_Male/FX/T_FluidFlameTile}.R) {constExponent=1}) {constB=0.1}
  n3 = TextureCoordinate() {uTiling=2, vTiling=0.2}
  n4 = TextureSample(UVs=Panner(Coordinate=@n3) {speedY=0.3}) {samplerType=SAMPLERTYPE_Masks, texture=/Game/Weapons/WeaponSkins/Wraps/Textures/Mash/Mash_Noise}
  n5 = TextureSample(UVs=Panner(Coordinate=@n3) {speedY=0.5}) {texture=&T_Inky_Smoke_Tile}
  n6 = Add(A=Subtract(A=Multiply(A=@n4.R, B=@n5.R)) {constB=0}) {constB=0}
  n7 = TextureCoordinate()
  n8 = Saturate(None=Multiply(A=ComponentMask(None=@n7) {g=true}) {constB=3})
  n9 = ComponentMask(None=TextureCoordinate()) {g=true}
  n10 = Multiply(A=Power(Base=OneMinus(None=@n9))) {constB=3}
  Multiply(A=@n1.A, B=Multiply(A=Multiply(A=@n6, B=@n8), B=Saturate(None=@n10)))
 ]
}
